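# Physics asset for the SphericalDefense two-handed pickaxe, giving its tassel bones capsule bodies and joint limits.
# One long capsule covers the handle, which anchors three-bone top and bottom tassel chains plus front, back, left and right middle tassels.
# Left and right middle tassels have two segments, front and back one.
# Top and bottom roots allow a wide swing on one axis, while the other joints are tightly limited.
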PhysicsAsset /BRCosmetics/Weapons/FORT_Melee/Pickaxe_SphericalDefense/Meshes/SphericalDefense_Physics
bodies {
 Melee_TwoHanded_Handle [{shapeType=Capsule, length=94.9187, radius=2.0648, center=<0.3897 -0.5342 40.3436>, rotation=<rot 0.3037 -89.9998 0.0083>}]
 bot_top_tassel_01 [{shapeType=Capsule, length=2.5539, radius=2.163, shapeName=bot_top_tassel_01_capsule, rotation=<rot -3.8011 -90 -1.3641>}]
 bot_top_tassel_02 [{
   shapeType=Capsule
   length=2.8333
   radius=1.7988
   shapeName=bot_top_tassel_02_capsule
   center=<-0.3525 -0.0321 -1.1441>
   rotation=<rot 0 175.4762 -88.2147>
  }]
 bot_top_tassel_03 [{
   shapeType=Capsule
   length=3.4773
   radius=3.0573
   shapeName=bot_top_tassel_03_capsule
   center=<0.0484 0.2244 -0.4969>
   rotation=<rot 0 -175.9225 -91.1804>
  }]
 dyn_mid_tassel_back_01 [{
   shapeType=Capsule
   length=8.079
   radius=1.6869
   shapeName=dyn_mid_tassel_back_01_capsule
   center=<6.0649 1.6834 0.0659>
   rotation=<rot -89.3778 0.0001 0>
  }]
 dyn_mid_tassel_front_01 [{
   shapeType=Capsule
   length=7.0856
   radius=2
   shapeName=dyn_mid_tassel_front_01_capsule
   center=<5.7777 -0.0081 -0.1404>
   rotation=<rot 0.4382 -90.0694 -88.6075>
  }]
 dyn_mid_tassel_left_01 [{
   shapeType=Capsule
   length=9.3938
   radius=2.0675
   shapeName=dyn_mid_tassel_left_01_capsule
   center=<5.4131 1.1877 -2.5722>
   rotation=<rot 88.2065 -171.7467 177.1031>
  }]
 dyn_mid_tassel_left_02 [{
   shapeType=Capsule
   length=4.0825
   radius=0.6961
   shapeName=dyn_mid_tassel_left_02_capsule
   center=<3.2871 1.0126 0.4698>
   rotation=<rot -0.1427 -57.4384 -98.1038>
  }]
 dyn_mid_tassel_right_01 [{
   shapeType=Capsule
   length=8.1694
   radius=1.743
   shapeName=dyn_mid_tassel_right_01_capsule
   center=<5.6977 -1.323 0.8276>
   rotation=<rot 1.4837 -103.2819 -81.8045>
  }]
 dyn_mid_tassel_right_02 [{
   shapeType=Capsule
   length=5.2943
   radius=0.8568
   shapeName=dyn_mid_tassel_right_02_capsule
   center=<4.1594 -1.2219 -1.4029>
   rotation=<rot 30.4369 -106.4833 112.1035>
  }]
 dyn_top_tassel_01 [{
   shapeType=Capsule
   length=2.4167
   radius=1.9492
   shapeName=dyn_top_tassel_01_capsule
   center=<-0.0308 -0.0363 -2.2288>
   rotation=<rot -0.8494 -90 -0.9858>
  }]
 dyn_top_tassel_02 [{
   shapeType=Capsule
   length=1.5688
   radius=3.7421
   shapeName=dyn_top_tassel_02_capsule
   center=<0.3223 0.1237 -4.0967>
   rotation=<rot 0.0011 176.801 -4.0464>
  }]
 dyn_top_tassel_03 [{
   shapeType=Capsule
   length=4.7697
   radius=4.3291
   shapeName=dyn_top_tassel_03_capsule
   center=<0.2036 -0.2249 -4.8914>
   rotation=<rot 0 1.7121 -3.0191>
  }]
}
constraints [
 {bone1Name=bot_top_tassel_01, bone2Name=Melee_TwoHanded_Handle, swing1LimitDegrees=45, swing2LimitDegrees=90, twistLimitDegrees=45}
 {bone1Name=bot_top_tassel_02, bone2Name=bot_top_tassel_01, swing1LimitDegrees=5, swing2LimitDegrees=5, twistLimitDegrees=5}
 {bone1Name=bot_top_tassel_03, bone2Name=bot_top_tassel_02, swing1LimitDegrees=10, swing2LimitDegrees=10, twistLimitDegrees=10}
 {bone1Name=dyn_top_tassel_01, bone2Name=Melee_TwoHanded_Handle, swing1LimitDegrees=5, swing2LimitDegrees=90, twistLimitDegrees=5}
 {bone1Name=dyn_top_tassel_02, bone2Name=dyn_top_tassel_01, swing1LimitDegrees=5, swing2LimitDegrees=5, twistLimitDegrees=5}
 {bone1Name=dyn_top_tassel_03, bone2Name=dyn_top_tassel_02, swing1LimitDegrees=10, swing2LimitDegrees=10, twistLimitDegrees=10}
 {bone1Name=dyn_mid_tassel_back_01, bone2Name=Melee_TwoHanded_Handle, swing1LimitDegrees=15, swing2LimitDegrees=8.6, twistLimitDegrees=15}
 {bone1Name=dyn_mid_tassel_front_01, bone2Name=Melee_TwoHanded_Handle, swing1LimitDegrees=6, swing2LimitDegrees=6, twistLimitDegrees=6}
 {bone1Name=dyn_mid_tassel_left_01, bone2Name=Melee_TwoHanded_Handle, swing1LimitDegrees=4, swing2LimitDegrees=19.84, twistLimitDegrees=4}
 {bone1Name=dyn_mid_tassel_left_02, bone2Name=dyn_mid_tassel_left_01, swing1LimitDegrees=16, swing2LimitDegrees=16, twistLimitDegrees=16}
 {bone1Name=dyn_mid_tassel_right_01, bone2Name=Melee_TwoHanded_Handle, swing1LimitDegrees=15, swing2LimitDegrees=12, twistLimitDegrees=15}
 {bone1Name=dyn_mid_tassel_right_02, bone2Name=dyn_mid_tassel_right_01, swing1LimitDegrees=10, swing2LimitDegrees=15, twistLimitDegrees=10}
]
